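# Looping ambient SoundCue for the creaking windmill spot at the CliffTown RustBucket POI, spatialized with occlusion, listener focus and limited concurrency.
# A random root optionally adds a short delay before two nested looping envelopes that slowly vary pitch and volume on different periods.
# Every quality level shares one random pick between two windmill creak loops and a three-layer vibrating-metal and landmark mix, with randomized pitch.
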SoundCue /Game/Sounds/Helios/POI/CliffTown/Spots/SC_Spot_RustBucket_Windmill_Creaking_Loop
root=Random_1
nodes [
 {id=Delay_0, children [Enveloper_1], properties {delayMax=0.3, delayMin=0.2}}
 {
  id=Enveloper_0
  children [QualityLevel_0]
  properties {
   bLoop=true
   bLoopIndefinitely=true
   loopEnd=20
   pitchCurve {editorCurveData {keys [{time=0}]}}
   volumeCurve {editorCurveData {keys [<at 0 1 RCIM_Cubic>,<at 10 0.9 RCIM_Cubic>,<at 20 1>]}}
  }
 }
 {
  id=Enveloper_1
  children [Enveloper_0]
  properties {
   bLoop=true
   bLoopIndefinitely=true
   loopEnd=30
   pitchCurve {editorCurveData {keys [{interpMode=RCIM_Cubic, time=0}, <at 15 0.9 RCIM_Cubic>, {time=30}]}}
   volumeCurve {editorCurveData {keys [<at 0 1>]}}
  }
 }
 {id=Mixer_0, children [WavePlayer_2,Modulator_1,WavePlayer_5], properties {inputVolume [3,1,1.5]}}
 {id=Mixer_1, children [Modulator_3], properties {inputVolume [0.8]}}
 {id=Modulator_0, children [WavePlayer_0], properties {pitchMax=1.1, pitchMin=0.9}}
 {id=Modulator_1, children [WavePlayer_3], properties {pitchMax=0.8, pitchMin=0.7}}
 {id=Modulator_2, children [WavePlayer_1], properties {pitchMax=1.1, pitchMin=0.9, volumeMax=0.8, volumeMin=0.7}}
 {id=Modulator_3, children [Mixer_0], properties {pitchMax=1.1, pitchMin=0.9}}
 {id=QualityLevel_0, children [Random_0,Random_0,Random_0,Random_0]}
 {id=Random_0, children [Modulator_0,Modulator_2,Mixer_1], properties {bShouldExcludeFromBranchCulling=true, weights [1,1,1]}}
 {id=Random_1, children [Delay_0,Enveloper_1], properties {weights [1,1]}}
 {id=WavePlayer_0, properties {bLooping=true, soundWaveAssetPtr=/Game/Sounds/Fort_Ambient_Looping/Wind_Exterior/windmill_creaking}}
 {id=WavePlayer_1, properties {bLooping=true, soundWaveAssetPtr=/Game/Sounds/Fort_Ambient_Looping/Wind_Exterior/windmill_creaking_opt}}
 {id=WavePlayer_2, properties {bLooping=true, soundWaveAssetPtr=/Game/Sounds/Apollo/POI/S15/Landmarks/Slurpy_Swamp_Keg_Building_Topper_Loop}}
 {
  id=WavePlayer_3
  properties {bLooping=true, soundWaveAssetPtr=/Game/Sounds/Artemis/POI/Blimps_Crashsites/Spots/Spot_Blimp_Crashsite_Wind_Vibrating_Metal_Large}
 }
 {
  id=WavePlayer_5
  properties {bLooping=true, soundWaveAssetPtr=/Game/Sounds/Artemis/POI/Hollow_Base/Spots/Spot_Wind_Excited_Metal_Vibrating_Loop}
 }
]
properties {
 attenuationSettings=/Game/Sounds/Attenuation/Fort_Prop_Med_Occlusion_Att
 bOverrideAttenuation=true
 bOverrideConcurrency=true
 maxDistance=3072
 soundClassObject=&Fort_Ambient_Props_Athena
 volumeMultiplier=0.301
 attenuationOverrides {
  bEnableListenerFocus=true
  bEnableOcclusion=true
  falloffDistance=2560
  nonFocusVolumeAttenuation=0.75
  occlusionInterpolationTime=0.3
  occlusionLowPassFilterFrequency=1000
  occlusionTraceChannel=ECC_GameTraceChannel11
  occlusionVolumeAttenuation=0.3
  stereoSpread=500
  attenuationShapeExtents {x=512}
  customAttenuationCurve {editorCurveData {keys [<at 0 1>, {arriveTangent=-4.0197, interpMode=RCIM_Cubic, leaveTangent=-4.0197, tangentMode=RCTM_User, time=0.5, value=0.7}, <at 1 0>]}}
 }
 concurrencyOverrides {maxCount=3}
}
