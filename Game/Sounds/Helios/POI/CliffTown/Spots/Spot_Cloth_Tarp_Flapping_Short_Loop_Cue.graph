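# Ambient spot sound cue for a short flapping cloth tarp loop in the CliffTown POI.
# A looping slow banner flap mixes with one of six random flap one-shots, retriggered after a few seconds' delay.
# Nested looping envelopers modulate pitch and swell volume over the mix, behind a short random start delay.
# It uses box attenuation, limits concurrency, and sends to exterior ambient bus and reverb submix.
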SoundCue /Game/Sounds/Helios/POI/CliffTown/Spots/Spot_Cloth_Tarp_Flapping_Short_Loop_Cue
root=Delay_0
nodes [
 {id=Delay_0, children [Enveloper_3], properties {delayMax=2}}
 {id=Delay_1, children [Random_0], properties {delayMax=6, delayMin=4}}
 {
  id=Enveloper_0
  children [Mixer_0]
  properties {
   bLoop=true
   bLoopIndefinitely=true
   loopEnd=60
   pitchCurve {editorCurveData {keys [
      {interpMode=RCIM_Cubic, time=0}
      <at 15 0.95 RCIM_Cubic>
      {arriveTangent=0.0072, interpMode=RCIM_Cubic, leaveTangent=0.0072, tangentMode=RCTM_User, time=30}
      {arriveTangent=-0.0082, interpMode=RCIM_Cubic, leaveTangent=-0.0082, tangentMode=RCTM_User, time=45}
      {time=60}
     ]}}
   volumeCurve {editorCurveData {keys [<at 0 1>]}}
  }
 }
 {
  id=Enveloper_3
  children [Enveloper_0]
  properties {
   bLoop=true
   bLoopIndefinitely=true
   loopEnd=20
   pitchCurve {editorCurveData {keys [{time=0}]}}
   volumeCurve {editorCurveData {keys [
      {arriveTangent=-0.01, interpMode=RCIM_Cubic, time=0, value=1}
      {arriveTangent=-0.1116, interpMode=RCIM_Cubic, leaveTangent=-0.1116, tangentMode=RCTM_User, time=5, value=0.8}
      {arriveTangent=0.0002, interpMode=RCIM_Cubic, leaveTangent=0.0002, tangentMode=RCTM_User, time=7.5, value=0.2}
      {arriveTangent=-0.0014, interpMode=RCIM_Cubic, leaveTangent=0.0091, tangentMode=RCTM_User, time=12.5, value=1}
      {arriveTangent=0.0016, interpMode=RCIM_Cubic, leaveTangent=0.0016, time=15.8677, value=0.4744}
      <at 20 1.0122 RCIM_Cubic>
     ]}}
  }
 }
 {id=Looping_0, children [Delay_1]}
 {id=Mixer_0, children [WavePlayer_2,Looping_0], properties {inputVolume [1,0.9]}}
 {id=Random_0, children [WavePlayer_0,WavePlayer_1,WavePlayer_3,WavePlayer_4,WavePlayer_5,WavePlayer_6], properties {weights [1,1,1,1,1,1]}}
 {id=WavePlayer_0, properties {soundWaveAssetPtr=/Game/Sounds/Helios/Environment_Actors/FlagBanner/Spot_Cloth_Flap_Oneshot_01}}
 {id=WavePlayer_1, properties {soundWaveAssetPtr=/Game/Sounds/Helios/Environment_Actors/FlagBanner/Spot_Cloth_Flap_Oneshot_02}}
 {
  id=WavePlayer_2
  properties {bLooping=true, soundWaveAssetPtr=/Game/Sounds/Helios/Environment_Actors/FlagBanner/Spot_Flag_Flap_Cloth_Banner_Slow_Loop}
 }
 {id=WavePlayer_3, properties {soundWaveAssetPtr=/Game/Sounds/Helios/Environment_Actors/FlagBanner/Spot_Cloth_Flap_Oneshot_03}}
 {id=WavePlayer_4, properties {soundWaveAssetPtr=/Game/Sounds/Helios/Environment_Actors/FlagBanner/Spot_Cloth_Flap_Oneshot_04}}
 {id=WavePlayer_5, properties {soundWaveAssetPtr=/Game/Sounds/Helios/Environment_Actors/FlagBanner/Spot_Cloth_Flap_Oneshot_05}}
 {id=WavePlayer_6, properties {soundWaveAssetPtr=/Game/Sounds/Helios/Environment_Actors/FlagBanner/Spot_Cloth_Flap_Oneshot_06}}
]
properties {
 attenuationSettings=&/Spot_Cloth_Tarp_Flapping_Short_Box_Att
 bOverrideConcurrency=true
 maxDistance=1150
 soundClassObject=&Fort_Ambient_Athena
 volumeMultiplier=0.501
 attenuationOverrides {
  attenuationShape=Box
  bEnableFocusInterpolation=true
  bEnableListenerFocus=true
  bEnableOcclusion=true
  dBAttenuationAtMax=-40
  distanceAlgorithm=Logarithmic
  falloffDistance=1500
  focusAzimuth=15
  focusVolumeAttenuation=1.1
  nonFocusAzimuth=45
  nonFocusVolumeAttenuation=0.4
  nonSpatializedRadiusStart=100
  occlusionTraceChannel=ECC_GameTraceChannel11
  occlusionVolumeAttenuation=0.7
  stereoSpread=2000
  attenuationShapeExtents {x=100}
 }
 busSends [{audioBus=null, minSendLevel=1, sendLevel=1, soundSourceBus=&AmbientAudio_Exterior_SourceBus}]
 concurrencyOverrides {maxCount=2, voiceStealReleaseTime=1}
 soundSubmixSends [{
   minSendLevel=0.4
   soundSubmix=&Fortnite_Reverb_Default
   customSendLevelCurve {editorCurveData {keys [<at 0 0.4>,<at 1 1>]}}
  }]
}
